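# Sound cue for the Ninja smoke bomb throw, built from reused Phantom Dash waves pitched up to double.
# A distance crossfade blends a near layer, a random pick of two waves, into a quieter far layer, a random pick of three mono-named waves.
# It uses skill-use attenuation and the melee weapon sound class, and overrides concurrency to cap instances at three, stopping the oldest.
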SoundCue /Game/Sounds/Fort_Gadgets/Ninja/Ninja_SmokeBomb_Throw_Cue
root=DistanceCrossFade_2
nodes [
 {
  id=DistanceCrossFade_2
  children [Modulator_2,Modulator_3]
  properties {crossFadeInput [
    {fadeInDistanceEnd=0, fadeInDistanceStart=0, fadeOutDistanceEnd=800, fadeOutDistanceStart=400, volume=1}
    {fadeInDistanceEnd=400, fadeInDistanceStart=200, fadeOutDistanceEnd=7000, fadeOutDistanceStart=5000, volume=0.7}
   ]}
 }
 {id=Modulator_2, children [Random_1], properties {pitchMax=2, pitchMin=2}}
 {id=Modulator_3, children [Random_0], properties {pitchMax=2, pitchMin=2}}
 {id=Random_0, children [WavePlayer_0,WavePlayer_1,WavePlayer_2], properties {weights [1,1,1]}}
 {id=Random_1, children [WavePlayer_34, WavePlayer_33], properties {weights [1,1]}}
 {id=WavePlayer_0, properties {soundWaveAssetPtr=&/phantom_dash_03_mono}}
 {id=WavePlayer_1, properties {soundWaveAssetPtr=&/phantom_dash_05_mono}}
 {id=WavePlayer_2, properties {soundWaveAssetPtr=&/phantom_dash_06_mono}}
 {id=WavePlayer_33, properties {soundWaveAssetPtr=/Game/Sounds/Fort_Gadgets/PhantomDash/phantom_dash_03}}
 {id=WavePlayer_34, properties {soundWaveAssetPtr=/Game/Sounds/Fort_Gadgets/PhantomDash/phantom_dash_05}}
]
properties {
 attenuationSettings=/Game/Sounds/Attenuation/Fort_SkillUse_Att
 bDebug=true
 bOverrideConcurrency=true
 maxDistance=7000
 soundClassObject=&Fort_WeaponMelee_PSM
 volumeMultiplier=0.4
 concurrencyOverrides {maxCount=3, resolutionRule=StopOldest}
}
